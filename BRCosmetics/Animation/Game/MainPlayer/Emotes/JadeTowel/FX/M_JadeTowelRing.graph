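# Translucent, two-sided particle material for the ring effect in the JadeTowel emote, driven by particle color and dynamic parameters.
# Color combines screen-space panning block noise, scanlines, chromatic offset samples and a normal-tinted Fresnel fake chroma.
# Opacity uses a local-Z mask broken up by triplanar pixel noise, while two panning waveform textures displace vertices along normals, masked by vertex red.
# A TriMode static switch swaps in alternate emissive and depth-faded opacity and disables displacement.
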Material /BRCosmetics/Animation/Game/MainPlayer/Emotes/JadeTowel/FX/M_JadeTowelRing
graph {
 properties {bAutomaticallySetUsageInEditor=false, blendMode=BLEND_Translucent, twoSided=true, usedShadingModels=Default Lit}
 nodes [
  Add()
  n1 = VertexNormalWS()
  Add(A=@n1, B=Time())
  n2 = ScreenPosition()
  n3 = ScalarParameter() {defaultValue=1, group=Block, parameterName=Block UV}
  n4 = DynamicParameter() {parameterIndex=2, defaultValue=<rgba 0 1 0 1>, paramNames [Block UV, WPO, FinalSaturation, Param4]}
  n5 = Time()
  n6 = ScalarParameter() {defaultValue=10.7749, group=Block, parameterName=Block U Speed}
  n7 = Panner(Coordinate=Multiply(A=@n2.ViewportUV, B=Add(A=@n3, B=@n4.Block UV)), Time=Ceil(None=Multiply(A=@n5, B=@n6) {constB=0.5})) {desc="-0.33,-0.43", speedX=-0.33, speedY=-0.43}
  n8 = TextureObject() {texture=/Game/Athena/UI/Challenges/Art/T_UI_DataChip_Noise}
  n9 = TextureSample(UVs=@n7, Tex=@n8) {texture=/Game/Effects/Fort_Effects/Textures/Flares/T_Flare_Pixel}
  n10 = Clamp(None=@n9.R)
  n11 = VectorParameter() {group=Block, parameterName=BlockColor1, defaultValue=<rgba 0 0.8127 1 1>}
  n12 = Multiply(A=@n10, B=@n11.RGB)
  n13 = Multiply(A=Constant2Vector() {g=0.05, r=0.05}, B=ScalarParameter() {defaultValue=0.2, parameterName=ChromOffset})
  n14 = Add(A=@n7, B=@n13)
  n15 = Subtract(A=@n7, B=@n13)
  n16 = LinearInterpolate(
   A=Multiply(
    A=VectorParameter() {parameterName=Chroma1, defaultValue=<rgba 1.5 0 0.8249 1>}.RGB
    B=TextureSample(UVs=@n14, Tex=@n8) {texture=&T_Flare_Pixel}.R
   )
   B=Multiply(
    A=TextureSample(UVs=@n15, Tex=@n8) {texture=&T_Flare_Pixel}.R
    B=VectorParameter() {parameterName=Chroma2, defaultValue=<rgba 0 1.5 1.2326 0>}.RGB
   )
  )
  Add(A=@n12, B=@n16)
  Add(A=@n9.R, B=@n9.R)
  Clamp(None=ParticleRandom()) {maxDefault=0.3, minDefault=0.1}
  Constant3Vector() {constant=<rgba 0 0.818 1 1>}
  Constant() {r=1}
  Constant()
  Constant()
  n17 = ParticleColor()
  n18 = DynamicParameter() {parameterIndex=3, defaultValue=<rgba 1 1 0.1 0.1>, paramNames [Mask, EdgeIntensity, EmisIntensity, Rand]}
  n19 = Multiply(A=@n17.RGB, B=@n18.EmisIntensity)
  n20 = NamedRerouteUsage()
  n21 = Add(A=LinearInterpolate(A=@n20) {constAlpha=0}) {constB=0}
  n22 = DynamicParameter() {parameterIndex=2, defaultValue=<rgba 0 1 0 1>, paramNames [Block UV, WPO, FinalSaturation, Param4]}
  n23 = MaterialFunctionCall(Float2=@n2.ViewportUV) {materialFunction=&BreakOutFloat2Components}
  n24 = ScalarParameter() {defaultValue=150, group=Block, parameterName=Line amount}
  n25 = Time()
  n26 = ScalarParameter() {defaultValue=0.2, group=Block, parameterName=Line Speed}
  n27 = Frac(None=Add(A=Multiply(A=@n23.G, B=@n24), B=Multiply(A=@n25, B=@n26)))
  n28 = Multiply(A=@n27, B=OneMinus(None=@n27))
  n29 = ScalarParameter() {defaultValue=-0.05, group=Block, parameterName=Line Size}
  n30 = LinearInterpolate(
   A=ScalarParameter() {group=Block, parameterName=Line Dark}
   B=ScalarParameter() {defaultValue=1, group=Block, parameterName=Line Light}
   Alpha=Ceil(None=Add(A=@n28, B=@n29))
  )
  n31 = Fresnel(ExponentIn=ScalarParameter() {defaultValue=2, parameterName=FakeChromExp}) {exponent=1}
  n32 = LinearInterpolate(
   A=SmoothStep(Value=@n31) {constMax=0, constMin=1}
   B=SmoothStep(Value=@n31)
   Alpha=ScalarParameter() {defaultValue=1, parameterName=FakeChromaInv}
  )
  n33 = Saturate(None=@n1)
  n34 = VectorParameter() {parameterName=Param_1, defaultValue=<rgba 1 1 1 1>}
  n35 = Multiply(A=Multiply(A=@n32, B=Multiply(A=@n33, B=@n34.RGB)), B=ScalarParameter() {defaultValue=1, parameterName=FakeChromVal}) {constB=50}
  n36 = LinearInterpolate(
   A=Desaturation(None=Multiply(A=@n19, B=@n21), Fraction=@n22.FinalSaturation)
   B=Multiply(A=@n30, B=Clamp(None=@n35) {maxDefault=5000})
   Alpha=Clamp(None=@n32)
  )
  n37 = ScalarParameter() {parameterName=UseTriEdgeMask}
  LinearInterpolate(A=@n36, B=Constant(), Alpha=@n37)
  MaterialFunctionCall()
  n38 = TextureObject() {texture=&/T_Emote_JadeTowel_PixelNoise}
  MaterialFunctionCall!(CameraVector, Texture=@n38, Tiling, Offset, AxisFadeContrast) {materialFunction=/Engine/Functions/Engine_MaterialFunctions01/Texturing/TriplanarCameraVector}
  n39 = ParticleColor()
  n40 = NamedRerouteUsage()
  n41 = LinearInterpolate(A=@n40)
  n42 = DynamicParameter() {parameterIndex=1, defaultValue=<rgba 0 1 0 1>, paramNames [Block UV, WPO, FinalSaturation, Param4]}
  n43 = Desaturation(None=Multiply(A=@n39.RGB, B=Add(A=@n41) {constB=0}), Fraction=@n42.FinalSaturation)
  MaterialFunctionCall(Emissive=@n43) {materialFunction=&MF_FX_EmissiveAdjustments}
  n44 = LinearInterpolate(A=Clamp(None=Subtract(A=@n16, B=@n10)), B=@n12, Alpha=Multiply(A=@n10))
  n45 = Multiply(A=@n30, B=@n44)
  Multiply(A=@n45, B=@n11.RGB)
  Multiply(B=Multiply())
  n46 = Multiply(A=@n30, B=@n44)
  Multiply(A=@n46, B=VectorParameter() {group=Block, parameterName=BlockColor2, defaultValue=<rgba 0 0.8225 1 1>}.RGB)
  n47 = LinearInterpolate(A=@n45, B=@n46)
  Multiply(A=@n47, B=ScalarParameter() {defaultValue=1, parameterName=BlockIntensity}) {constB=5}
  Multiply()
  n48 = DynamicParameter() {defaultValue=<rgba 1 -0.4 0.3 0.3>, paramNames [Amplitude,Rand1,Rand2,Bias]}
  NamedRerouteDeclaration(None=@n48.Rand1) {name=Rand1, nodeColor=<rgba 0.1767 0.0464 1 1>}
  NamedRerouteDeclaration(None=@n48.Rand2) {name=Rand2, nodeColor=<rgba 0 0.6719 1 1>}
  NamedRerouteDeclaration(None=@n48.Bias) {name=Bias, nodeColor=<rgba 0 1 0.6452 1>}
  NamedRerouteDeclaration(None=@n48.Amplitude) {name=Amplitude, nodeColor=<rgba 0.5469 0 1 1>}
  n49 = ScreenPosition()
  n50 = MaterialFunctionCall(Float2=@n49.ViewportUV) {materialFunction=&BreakOutFloat2Components}
  n51 = ScalarParameter() {defaultValue=150, group=Block, parameterName=Line amount}
  n52 = Time()
  n53 = ScalarParameter() {defaultValue=0.2, group=Block, parameterName=Line Speed}
  n54 = Frac(None=Add(A=Multiply(A=@n50.G, B=@n51), B=Multiply(A=@n52, B=@n53)))
  n55 = Multiply(A=@n54, B=OneMinus(None=@n54))
  n56 = ScalarParameter() {defaultValue=-0.05, group=Block, parameterName=Line Size}
  n57 = LinearInterpolate(
   A=ScalarParameter() {group=Block, parameterName=Line Dark}
   B=ScalarParameter() {defaultValue=1, group=Block, parameterName=Line Light}
   Alpha=Ceil(None=Add(A=@n55, B=@n56))
  )
  n58 = DynamicParameter() {parameterIndex=2, defaultValue=<rgba 0 1 0 1>, paramNames [Block UV, WPO, FinalSaturation, Param4]}
  n59 = Add(A=ScalarParameter() {defaultValue=1, group=Block, parameterName=Block UV}, B=@n58.Block UV)
  n60 = Multiply(A=Time(), B=ScalarParameter() {defaultValue=10.7749, group=Block, parameterName=Block U Speed}) {constB=0.5}
  n61 = TextureSample(UVs=Panner(Coordinate=Multiply(A=@n49.ViewportUV, B=@n59), Time=Ceil(None=@n60)) {desc="-0.33,-0.43", speedX=-0.33, speedY=-0.43}) {texture=&T_UI_DataChip_Noise}
  n62 = Multiply(A=@n57, B=LinearInterpolate(Alpha=@n61.R) {constA=0.01})
  n63 = VectorParameter() {group=Block, parameterName=BlockColor, defaultValue=<rgba 1 1 1 0>}
  n64 = Multiply(A=@n57, B=LinearInterpolate(Alpha=@n61.G) {constA=-0.5, constB=0.5})
  n65 = VectorParameter() {group=Block, parameterName=BlockColor2, defaultValue=<rgba 0 0.8225 1 1>}
  NamedRerouteDeclaration(None=LinearInterpolate(A=Multiply(A=@n62, B=@n63.RGB), B=Multiply(A=@n64, B=@n65.RGB))) {name=Block_1, nodeColor=<rgba 0 1 0.4844 1>}
  n66 = TextureSample(UVs=@n7) {texture=&T_UI_DataChip_Noise}
  n67 = Multiply(A=@n30, B=LinearInterpolate(Alpha=@n66.R) {constA=0.01})
  n68 = Multiply(A=@n30, B=LinearInterpolate(Alpha=@n66.G) {constA=-0.5, constB=0.5})
  NamedRerouteDeclaration(None=LinearInterpolate(A=@n47, B=LinearInterpolate(A=@n67, B=@n68), Alpha=@n37)) {name=Block, nodeColor=<rgba 0 1 0.4844 1>}
  n69 = ParticleColor()
  n70 = DynamicParameter() {parameterIndex=3, defaultValue=<rgba 1 1 0.1 0.1>, paramNames [Mask, EdgeIntensity, EmisIntensity, Rand]}
  n71 = Multiply(A=Add(A=Desaturation(None=@n40), B=@n70.Rand))
  n72 = Multiply(A=@n39.A, B=Saturate(None=Add(A=@n71, B=@n70.Rand)))
  NamedRerouteDeclaration(None=DepthFade(Opacity=Multiply(A=@n69.A, B=@n72)) {fadeDistanceDefault=500}) {name=Tri Opacity, nodeColor=<rgba 0.9922 1 0 1>}
  n73 = Multiply(A=Multiply(A=Multiply(A=@n69.RGB, B=@n70.EmisIntensity), B=@n70.EdgeIntensity), B=VertexColor().R)
  NamedRerouteDeclaration(None=MaterialFunctionCall(Emissive=Add(A=@n43, B=@n73)) {materialFunction=&MF_FX_EmissiveAdjustments}.Result) {name=Tri Emis, nodeColor=<rgba 0.6406 1 0 1>}
  ScalarParameter() {defaultValue=1, parameterName=Alpha}
  ScalarParameter() {defaultValue=1, group=Block, parameterName=BlockAlpha}
  ScalarParameter() {defaultValue=0.3}
  ScalarParameter() {defaultValue=1, parameterName=Alpha}
  ScalarParameter() {defaultValue=1, group=Block, parameterName=BlockAlpha}
  StaticSwitchParameter(
   True=NamedRerouteUsage()
   False=MaterialFunctionCall(Emissive=@n36) {materialFunction=&MF_FX_EmissiveAdjustments}.Result
  ) {parameterName=TriMode}
  n74 = Constant() {r=1}
  n75 = Add(
   A=@n20
   B=LinearInterpolate(A=VectorParameter() {defaultValue=<rgba 0.1188 1 0.9003 1>}.RGB, B=ScalarParameter() {defaultValue=0.4, parameterName=Param_1}, Alpha=@n37)
  )
  StaticSwitchParameter(False=Multiply(A=LinearInterpolate(A=@n74, B=@n17.RGB, Alpha=@n37), B=Clamp(None=@n75))) {parameterName=TriMode}
  n76 = ParticleColor()
  n77 = DynamicParameter() {parameterIndex=1, defaultValue=<rgba 1 2 0.1 0.1>, paramNames [Pan1Size,Pan2Size,BaseAlpha,BaseAlpha2]}
  n78 = Multiply(A=Add(A=Desaturation(None=@n20), B=@n77.BaseAlpha))
  n79 = Multiply(A=@n17.A, B=Saturate(None=Add(A=@n78, B=@n77.BaseAlpha2)))
  n80 = OneMinus(None=ComponentMask(None=LocalPosition().XYZ) {b=true})
  n81 = Saturate(None=Add(A=Divide(A=@n80) {constB=500}, B=@n18.Mask))
  n82 = ComponentMask(None=MaterialFunctionCall!(
    TextureObject=@n38
    TextureSize=Constant() {r=500}
    WorldPosition
    Export Float 4
    Parallax
    In
    World Space Normal
    ProjectionTransitionContrast
   ) {materialFunction=/BRCosmetics/Characters/Player/Male/Medium/Bodies/M_MED_SilverBellMarine/Materials/MF_Triplanar_ScalePan_Parallax}.XYZ Texture) {r=true}
  n83 = Clamp(None=Subtract(A=@n81, B=SmoothStep(Value=@n82)))
  StaticSwitchParameter(True=NamedRerouteUsage(), False=Multiply(A=Multiply(A=@n76.A, B=@n79), B=Add(A=@n81, B=@n83))) {parameterName=TriMode}
  n84 = Multiply(A=Multiply(A=VertexNormalWS())) {constB=4}
  n85 = VertexColor()
  n86 = TextureCoordinate() {coordinateIndex=1}
  n87 = Multiply(A=@n86, B=@n77.Pan1Size)
  n88 = NamedRerouteUsage()
  n89 = TextureSample(UVs=Panner(Coordinate=Add(A=@n87, B=@n88), Speed=@n88) {speedX=0.1}) {texture=&/T_Emote_Jade_WaveformMod}
  n90 = NamedRerouteUsage()
  n91 = Multiply(A=@n86, B=@n77.Pan2Size)
  n92 = NamedRerouteUsage()
  n93 = TextureSample(UVs=Panner(Coordinate=Add(A=@n91, B=@n92), Speed=@n92)) {texture=&T_Emote_Jade_WaveformMod}
  n94 = Multiply(A=Clamp(None=Add(A=@n89.R, B=@n90)), B=Clamp(None=Add(A=@n93.R, B=@n90)))
  n95 = NamedRerouteUsage()
  StaticSwitchParameter(True=Constant(), False=Multiply(A=Multiply(A=@n84, B=@n85.R), B=Multiply(A=@n94, B=@n95))) {parameterName=TriMode}
  TextureCoordinate()
  TextureCoordinate()
  TextureCoordinate()
  TextureSample() {texture=&T_Emote_JadeTowel_PixelNoise}
  VectorParameter() {parameterName=ColorTest, defaultValue=<rgba 0 0 0 1>}
  VertexTangentWS()
 ]
}
