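# Translucent, unlit, two-sided material for Niagara mesh particles, apparently the ring effect of the JadeTowel emote.
# A reversible local-axis mask is broken up by triplanar pixel noise, and a screen-space block mask combines scanlines with stepped noise.
# Fresnel masks a scrolling color curve for fake chromatic fringing.
# Two panned waveforms and red vertex color shape normal-direction vertex offset, with particle color and dynamic parameters supplying per-particle control.
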Material /BRCosmetics/Animation/Game/MainPlayer/Emotes/JadeTowel/FX/M_JadeTowelRing_New
graph {
 properties {
  bAutomaticallySetUsageInEditor=false
  bUsedWithNiagaraMeshParticles=true
  blendMode=BLEND_Translucent
  shadingModel=MSM_Unlit
  twoSided=true
  usedShadingModels=Unlit
 }
 nodes [
  n1 = TextureObject() {texture=&/T_Emote_JadeTowel_PixelNoise}
  MaterialFunctionCall!(CameraVector, Texture=@n1, Tiling, Offset, AxisFadeContrast) {materialFunction=/Engine/Functions/Engine_MaterialFunctions01/Texturing/TriplanarCameraVector}
  MaterialFunctionCall(Emissive=LinearInterpolate(A=NamedRerouteUsage(), B=NamedRerouteUsage(), Alpha=NamedRerouteUsage())) {materialFunction=&MF_FX_EmissiveAdjustments}
  n2 = ScreenPosition()
  n3 = MaterialFunctionCall(Float2=@n2.ViewportUV) {materialFunction=&BreakOutFloat2Components}
  n4 = ScalarParameter() {defaultValue=150, group=Block, parameterName=Line amount}
  n5 = Time()
  n6 = ScalarParameter() {defaultValue=0.2, group=Block, parameterName=Line Speed}
  n7 = Frac(None=Add(A=Multiply(A=@n3.G, B=@n4), B=Multiply(A=@n5, B=@n6)))
  n8 = Multiply(A=@n7, B=OneMinus(None=@n7))
  n9 = ScalarParameter() {defaultValue=-0.05, group=Block, parameterName=Line Size}
  n10 = LinearInterpolate(
   A=ScalarParameter() {group=Block, parameterName=Line Dark}
   B=ScalarParameter() {defaultValue=1, group=Block, parameterName=Line Light}
   Alpha=Ceil(None=Add(A=@n8, B=@n9))
  )
  n11 = DynamicParameter() {parameterIndex=2, defaultValue=<rgba 0 1 0 1>, paramNames [Block UV, WPO, FinalSaturation, ReverseMask]}
  n12 = Add(A=ScalarParameter() {defaultValue=1, group=Block, parameterName=Block UV}, B=@n11.Block UV)
  n13 = Multiply(A=Time(), B=ScalarParameter() {defaultValue=10.7749, group=Block, parameterName=Block U Speed}) {constB=0.5}
  n14 = TextureSample(UVs=Panner(Coordinate=Multiply(A=@n2.ViewportUV, B=@n12), Time=Ceil(None=@n13)) {desc="-0.33,-0.43", speedX=-0.33, speedY=-0.43}) {texture=/Game/Athena/UI/Challenges/Art/T_UI_DataChip_Noise}
  n15 = Multiply(A=@n10, B=Multiply(A=@n10, B=LinearInterpolate(Alpha=@n14.R) {constA=0.01}))
  Multiply(A=@n15, B=VectorParameter() {group=Block, parameterName=BlockColor1, defaultValue=<rgba 0 0.8127 1 1>}.RGB)
  n16 = Multiply(A=@n10, B=Multiply(A=@n10, B=LinearInterpolate(Alpha=@n14.G) {constA=-0.5, constB=0.5}))
  Multiply(A=@n16, B=VectorParameter() {group=Block, parameterName=BlockColor2, defaultValue=<rgba 0 0.8225 1 1>}.RGB)
  n17 = NamedRerouteUsage()
  n18 = NamedRerouteUsage()
  Multiply(A=Multiply(A=NamedRerouteUsage(), B=LinearInterpolate(B=@n17, Alpha=@n18) {constA=1}), B=NamedRerouteUsage())
  n19 = DynamicParameter() {defaultValue=<rgba 1 -0.4 0.3 0.3>, paramNames [Amplitude,Rand1,Rand2,Bias]}
  NamedRerouteDeclaration(None=@n19.Rand1) {name=Rand1, nodeColor=<rgba 0.1767 0.0464 1 1>}
  NamedRerouteDeclaration(None=@n19.Rand2) {name=Rand2, nodeColor=<rgba 0 0.6719 1 1>}
  n20 = Fresnel(ExponentIn=ScalarParameter() {defaultValue=2, parameterName=FakeChromExp}) {exponent=1}
  n21 = LinearInterpolate(
   A=SmoothStep(Value=@n20) {constMax=0, constMin=1}
   B=SmoothStep(Value=@n20)
   Alpha=ScalarParameter() {defaultValue=1, parameterName=FakeChromaInv}
  )
  NamedRerouteDeclaration(None=Clamp(None=@n21)) {name=Chrom Mask, nodeColor=<rgba 0 0.1562 1 1>}
  n22 = Multiply(A=Multiply(A=VertexNormalWS())) {constB=4}
  n23 = VertexColor()
  n24 = TextureCoordinate() {coordinateIndex=1}
  n25 = DynamicParameter() {parameterIndex=1, defaultValue=<rgba 1 2 0 0.1>, paramNames [Pan1Size,Pan2Size,BaseAlpha,BaseAlpha2]}
  n26 = Multiply(A=@n24, B=@n25.Pan1Size)
  n27 = NamedRerouteUsage()
  n28 = TextureSample(UVs=Panner(Coordinate=Add(A=@n26, B=@n27), Speed=@n27) {speedX=0.1}) {texture=&/T_Emote_Jade_WaveformMod}
  n29 = NamedRerouteUsage()
  n30 = Multiply(A=@n24, B=@n25.Pan2Size)
  n31 = NamedRerouteUsage()
  n32 = TextureSample(UVs=Panner(Coordinate=Add(A=@n30, B=@n31), Speed=@n31)) {texture=&T_Emote_Jade_WaveformMod}
  n33 = Multiply(A=Clamp(None=Add(A=@n28.R, B=@n29)), B=Clamp(None=Add(A=@n32.R, B=@n29)))
  n34 = NamedRerouteUsage()
  NamedRerouteDeclaration(None=Multiply(A=Multiply(A=@n22, B=@n23.R), B=Multiply(A=@n33, B=@n34))) {name=Ring WPO, nodeColor=<rgba 0 1 0.9531 1>}
  n35 = ComponentMask(None=LocalPosition().XYZ) {b=true}
  n36 = LinearInterpolate(A=OneMinus(None=@n35), B=@n35, Alpha=NamedRerouteUsage())
  n37 = DynamicParameter() {parameterIndex=3, defaultValue=<rgba 1 1 0.1 0.1>, paramNames [Mask, EdgeIntensity, EmisIntensity, Rand]}
  n38 = Saturate(None=Add(A=Divide(A=@n36) {constB=500}, B=@n37.Mask))
  n39 = MaterialFunctionCall!(
   TextureObject=@n1
   TextureSize=Constant() {r=500}
   WorldPosition
   Export Float 4
   Parallax
   In
   World Space Normal
   ProjectionTransitionContrast
  ) {materialFunction=/BRCosmetics/Characters/Player/Male/Medium/Bodies/M_MED_SilverBellMarine/Materials/MF_Triplanar_ScalePan_Parallax}
  n40 = Subtract(A=@n38, B=SmoothStep(Value=ComponentMask(None=@n39.XYZ Texture) {r=true}))
  NamedRerouteDeclaration(None=Add(A=@n38, B=Clamp(None=@n40))) {name=Mask, nodeColor=<rgba 0.9688 0 1 1>}
  NamedRerouteDeclaration(None=@n19.Bias) {name=Bias, nodeColor=<rgba 0 1 0.6452 1>}
  NamedRerouteDeclaration(None=@n19.Amplitude) {name=Amplitude, nodeColor=<rgba 0.5469 0 1 1>}
  NamedRerouteDeclaration(None=@n25.BaseAlpha) {name=BlockMaskVal, nodeColor=<rgba 0.8281 1 0 1>}
  NamedRerouteDeclaration(None=LinearInterpolate(A=@n15, B=@n16)) {name=BlockMask, nodeColor=<rgba 0 1 0.4844 1>}
  NamedRerouteDeclaration(None=@n11.ReverseMask) {name=ReverseMask, nodeColor=<rgba 1 0.5156 0 1>}
  n41 = ParticleColor()
  NamedRerouteDeclaration(None=@n41.A) {name=Alpha, nodeColor=<rgba 0.6406 0 1 1>}
  NamedRerouteDeclaration(None=@n41.RGB) {name=Emis, nodeColor=<rgba 0 1 0.6484 1>}
  n42 = Add(A=ComponentMask(None=TextureCoordinate() {coordinateIndex=1}) {r=true}, B=Multiply(A=Time()) {constB=0.1})
  n43 = Multiply(
   A=@n21
   B=MaterialFunctionCall(
    Hue Shift Percentage=ScalarParameter() {parameterName=Param_2}
    Texture=TextureSample(UVs=@n42) {texture=&/T_Emote_JadeTowel_ColorCurve3}.RGB
   ) {materialFunction=&HueShift}.Result
  )
  n44 = ScalarParameter() {defaultValue=1, parameterName=FakeChromVal}
  NamedRerouteDeclaration(None=Clamp(None=Multiply(A=@n43, B=@n44) {constB=50}) {maxDefault=5000}) {name=Chrom, nodeColor=<rgba 0.3125 1 0 1>}
  NamedRerouteUsage()
  Saturate(None=VertexNormalWS())
  ScalarParameter() {defaultValue=0.3}
  TextureCoordinate()
  TextureSample() {texture=&T_Emote_JadeTowel_PixelNoise}
 ]
}
