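# Translucent, two-sided material for Niagara mesh particles in the JadeTowel emote, giving a glitchy digital-block look.
# Screen-space scrolling scanlines modulate two channels of noise, panned in stepped time, to build the colored Block pattern.
# The pattern, tinted by particle color, feeds emissive, with an edge glow masked by vertex red.
# Its desaturated value drives depth-faded opacity, and dynamic parameters adjust intensity, saturation and block scale.
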Material /BRCosmetics/Animation/Game/MainPlayer/Emotes/JadeTowel/FX/M_JadeTowel_Tri
graph {
 properties {
  bAutomaticallySetUsageInEditor=false
  bUsedWithNiagaraMeshParticles=true
  blendMode=BLEND_Translucent
  twoSided=true
  usedShadingModels=Default Lit
 }
 nodes [
  Clamp(None=ParticleRandom()) {maxDefault=0.3, minDefault=0.1}
  n1 = ParticleColor()
  n2 = ParticleColor()
  n3 = NamedRerouteUsage()
  n4 = DynamicParameter() {defaultValue=<rgba 1 1 0.1 0.1>, paramNames [EdgeIntensity, EmisIntensity, Param3, Param4]}
  n5 = Add(A=Desaturation(None=@n3), B=@n4.Param3)
  n6 = Saturate(None=Add(A=Multiply(A=@n5), B=@n4.Param4))
  DepthFade(Opacity=Multiply(A=@n1.A, B=Multiply(A=@n2.A, B=@n6))) {fadeDistanceDefault=500}
  n7 = LinearInterpolate(A=@n3)
  n8 = DynamicParameter() {parameterIndex=1, defaultValue=<rgba 0 1 0 1>, paramNames [Block UV, WPO, FinalSaturation, Param4]}
  n9 = Desaturation(None=Multiply(A=@n2.RGB, B=Add(A=@n7)), Fraction=@n8.FinalSaturation)
  n10 = Multiply(A=Multiply(A=@n1.RGB, B=@n4.EmisIntensity), B=@n4.EdgeIntensity)
  n11 = VertexColor()
  MaterialFunctionCall(Emissive=Add(A=@n9, B=Multiply(A=@n10, B=@n11.R))) {materialFunction=&MF_FX_EmissiveAdjustments}
  MaterialFunctionCall(Emissive=@n9) {materialFunction=&MF_FX_EmissiveAdjustments}
  n12 = ScreenPosition()
  n13 = MaterialFunctionCall(Float2=@n12.ViewportUV) {materialFunction=&BreakOutFloat2Components}
  n14 = ScalarParameter() {defaultValue=150, group=Block, parameterName=Line amount}
  n15 = Time()
  n16 = ScalarParameter() {defaultValue=0.2, group=Block, parameterName=Line Speed}
  n17 = Frac(None=Add(A=Multiply(A=@n13.G, B=@n14), B=Multiply(A=@n15, B=@n16)))
  n18 = Multiply(A=@n17, B=OneMinus(None=@n17))
  n19 = ScalarParameter() {defaultValue=-0.15, group=Block, parameterName=Line Size}
  n20 = LinearInterpolate(
   A=ScalarParameter() {group=Block, parameterName=Line Dark}
   B=ScalarParameter() {defaultValue=1, group=Block, parameterName=Line Light}
   Alpha=Ceil(None=Add(A=@n18, B=@n19))
  )
  n21 = DynamicParameter() {parameterIndex=1, defaultValue=<rgba 0 1 0 1>, paramNames [Block UV, WPO, FinalSaturation, Param4]}
  n22 = Add(A=ScalarParameter() {defaultValue=1, group=Block, parameterName=Block UV}, B=@n21.Block UV)
  n23 = Multiply(A=Time(), B=ScalarParameter() {defaultValue=10.7749, group=Block, parameterName=Block U Speed}) {constB=0.5}
  n24 = TextureSample(UVs=Panner(Coordinate=Multiply(A=@n12.ViewportUV, B=@n22), Time=Ceil(None=@n23)) {desc="-0.33,-0.43", speedX=-0.33, speedY=-0.43}) {texture=/Game/Athena/UI/Challenges/Art/T_UI_DataChip_Noise}
  n25 = Multiply(A=@n20, B=LinearInterpolate(Alpha=@n24.R) {constA=0.01})
  n26 = VectorParameter() {group=Block, parameterName=BlockColor, defaultValue=<rgba 1 1 1 0>}
  n27 = Multiply(A=@n20, B=LinearInterpolate(Alpha=@n24.G) {constA=-0.5, constB=0.5})
  n28 = VectorParameter() {group=Block, parameterName=BlockColor2, defaultValue=<rgba 1 1 1 0>}
  NamedRerouteDeclaration(None=LinearInterpolate(A=Multiply(A=@n25, B=@n26.RGB), B=Multiply(A=@n27, B=@n28.RGB))) {name=Block, nodeColor=<rgba 0 1 0.4844 1>}
  ScalarParameter() {parameterName=Alpha}
  ScalarParameter() {group=Block, parameterName=BlockAlpha}
  TextureCoordinate()
 ]
}
