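# Niagara system for the JadeTowel emote's audio-reactive lighting effect, built from two burst emitters.
# Minimal002 places three elevated meshes with lights at fixed spawn-group offsets, shifting color with event audio amplitude and adjusting one group's alpha by time of day.
# Minimal003 creates two mirrored, swaying mesh-and-sprite fixtures that grow in Z, with color and glow driven by music spectrum.
# Emitter-age curves time the intro flash; several alternative color modules are disabled.
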NiagaraSystem /BRCosmetics/Animation/Game/MainPlayer/Emotes/JadeTowel/FX/NS_JadeTowelLights
dependencies {
 usedDataInterfaces [
  &NiagaraDataInterfaceArrayFloat3
  &NiagaraDataInterfaceColorCurve
  &NiagaraDataInterfaceCurve
  &NiagaraDataInterfaceArrayFloat
  &NiagaraDataInterfaceArrayColor
 ]
 usedDynamicInputs [
  &MakeCustomFloatFromBool
  &SetBoolByFloatComparison
  &SelectLinearColorFromArray
  &RandomRangeInt
 ]
 usedModules [
  &SystemState
  &InterpolateOverTime
  &self:SystemSpawnScript.SystemScriptSource.SystemScriptGraph.NiagaraNodeAssignment_0.SetVariables_#1
  &self:SystemSpawnScript.SystemScriptSource.SystemScriptGraph.NiagaraNodeAssignment_1.SetVariables_#2
  &self:SystemSpawnScript.SystemScriptSource.SystemScriptGraph.NiagaraNodeAssignment_2.SetVariables_#3
  &self:SystemSpawnScript.SystemScriptSource.SystemScriptGraph.NiagaraNodeAssignment_3.SetVariables_#4
  &self:SystemSpawnScript.SystemScriptSource.SystemScriptGraph.NiagaraNodeAssignment_4.SetVariables_#5
 ]
}
emitters {
 Minimal002 {
  emitterUpdateScript [
   &EmitterState(
    Inactive Response=::NewEnumerator1
    Life Cycle Mode=::NewEnumerator1
    Loop Behavior=::NewEnumerator1
    Loop Duration Mode=::NewEnumerator1
    UseLoopDelay=false
   )
   &SpawnRate(Use Spawn Probability=false) {enabled=false}
   &SpawnBurst_Instantaneous(Spawn Count=3, Use Spawn Probability=false, UseLoopCountLimit=false)
  ]
  particleSpawnScript [
   &FN_SetSpawnGroup(Number of Spawn Groups=3)
   &InitializeParticle(
    Color=<rgba 0 0.9186 1 1>
    Lifetime Min=2.5
    Lifetime Mode=::NewEnumerator1
    Mesh Scale=&SelectVectorFromArray(
     Direct Array Index=$Particles.SpawnGroup
     Vector Selection Array=DataInterface() {propertyValues {FloatData [<0.6 0.6 1.4>, <0.6 0.6 1.4>, <0.6 0.6 1.4>]}}
    )
    Mesh Scale Mode=::NewEnumerator4
    Position Offset=&SelectVectorFromArray(
     Direct Array Index=$Particles.SpawnGroup
     Vector Selection Array=DataInterface() {propertyValues {FloatData [<-89.7052 -69.9752 250>, <0 -95.4906 250>, <105.9833 -49.4759 250>]}}
    )
    Sprite Size Mode=::NewEnumerator3
    Uniform Sprite Size=500
    UsePositionOffset=true
    Write Material Random=true
   )
   &InitialMeshOrientation(
    Mesh Orientation Mode=::NewEnumerator2
    Rotation=&AddVectorToPosition(
     Position=&SelectVectorFromArray(
      Direct Array Index=$Particles.SpawnGroup
      Vector Selection Array=DataInterface() {propertyValues {FloatData [<0.016 -0.042 0>, <0.0211 -0.012 0>, <0.06 0.0511 0>]}}
     )
     Vector=<0 0 0>
    )
    Use Rotation Vector=true
   )
   &CameraOffset(Camera Offset Amount=-20)
  ]
  particleUpdateScript [
   &ParticleState(Kill Particles When Lifetime Has Elapsed=false, Let Infinitely Lived Particles Die When Emitter Deactivates=true)
   &Color(Color=&ColorFromCurve(
     ColorCurve=DataInterface() {propertyValues {
       AlphaCurve {keys [<at 0 7 RCIM_Cubic>]}
       BlueCurve {keys [<at 0.25 4 RCIM_Cubic>,<at 0.35 4>,<at 0.45 4>,<at 0.5 4>,<at 0.6 4>,<at 0.7 4>]}
       GreenCurve {keys [<at 0.25 3.701 RCIM_Cubic>,<at 0.35 3.8416>,<at 0.45 2.8593>,<at 0.5 2.254>,<at 0.6 3.52>,<at 0.7 3.6812>]}
       RedCurve {keys [<at 0.25 0 RCIM_Cubic>,<at 0.35 1.375>,<at 0.45 0.8199>,<at 0.5 0.4184>,<at 0.6 1.1639>,<at 0.7 0.5671>]}
      }}
     CurveIndex=$Emitter.NormalizedLoopAge
    )) {enabled=false, module=Color001}
   &ScaleColor(
    Color Value To Scale=$Particles.Color
    Scale Alpha=&Multiply_Float(
     A=&FloatFromCurve(
      FloatCurve=DataInterface() {propertyValues {Curve {keys [
          {interpMode=RCIM_Cubic, tangentMode=RCTM_User, time=0, value=0}
          {interpMode=RCIM_Cubic, tangentMode=RCTM_User, time=0.5, value=1}
          {tangentMode=RCTM_User, time=1, value=0}
         ]}}}
      Scale Curve=0.025
     )
     B=&Multiply_Float(
      A=&Lerp_Float(A=$User.Other Alpha Preview, B=$User.Local Alpha Preview, Alpha=$NPC.FortniteNPC.EventAudio_Amplitude)
      B=&FN_FrontEndFloatModifier(Gameplay=0.5, Front End=0.3)
     )
    )
    Scale RGB=&VectorFromFloat(Value=&FloatFromCurve(
      FloatCurve=DataInterface() {propertyValues {Curve {keys [
          {interpMode=RCIM_Cubic, tangentMode=RCTM_Break, time=0.4, value=1}
          {interpMode=RCIM_Cubic, tangentMode=RCTM_Break, time=0.5, value=3}
          {interpMode=RCIM_Cubic, tangentMode=RCTM_Break, time=0.6, value=1}
          <at 0.77 1>
          <at 0.82 1>
         ]}}}
      CurveIndex=&Add_Float(
       A=$Particles.NormalizedAge
       B=&SelectFloatFromArray(
        Direct Array Index=&SpawnGroup()
        Float Selection Array=DataInterface() {propertyValues {FloatData [0.03,0.04,0.025,0.02,0.035,0.027]}}
       )
      )
      Scale Curve=&FN_FrontEndFloatModifier(Gameplay=1, Front End=1.5)
     ))
    ScaleA=true
    ScaleRGB=true
   ) {enabled=false, module=ScaleColor001}
   /SetVariables(Particles.Linear Color=&Multiply_LinearColorByFloat(Linear Color=$Particles.Color, Float=100)) {isSetParameters=true}
   &Color(Color=&LerpLinearColors(
     StartColor=<rgba 0 0.5318 1 0.025>
     EndColor=<rgba 0 1 0.8627 0.08>
     LerpFactor=&RemapRange(Input Value To Remap=$NPC.FortniteNPC.EventAudio_Amplitude, Input Min=0.6, Input Max=1, Output Min=0, Output Max=1)
    ))
   /SetVariables(Particles.Color=/Niagara/DynamicInputs/Add/Add_LinearColor(
     A:LinearColor=$Particles.Color
     B:LinearColor=&MakeLinearColor(
      R:NiagaraFloat=0
      G:NiagaraFloat=0
      B:NiagaraFloat=0
      A:NiagaraFloat=&FloatFromCurve(
       FloatCurve=DataInterface() {propertyValues {Curve {keys [<at 0.68 -0.2>, <at 0.75 0.4>, {interpMode=RCIM_Cubic, leaveTangent=8.5714, time=1, value=0.1}]}}}
       CurveIndex=$Emitter.Age
      )
     )
    )) {isSetParameters=true}
   /SetVariables(Particles.Color=&Add_LinearColor(
     A:LinearColor=$Particles.Color
     B:LinearColor=&MaskLinearColorBySpawnGroup(
      Masked Linear Color=&MakeLinearColor(
       R:NiagaraFloat=0
       G:NiagaraFloat=0
       B:NiagaraFloat=0
       A:NiagaraFloat=&FNTimeOfDayScalar(Night Scale=-0.15, Morning Scale=-0.14, Day Scale=-0.14, Evening Scale=-0.14)
      )
      Passthrough Linear Color=<rgba 0 0 0 0>
      Spawn Group Mask=1
     )
    )) {isSetParameters=true}
  ]
  rendererClasses [&NiagaraMeshRendererProperties, &NiagaraLightRendererProperties]
  renderers [
   {rendererClass=&NiagaraMeshRendererProperties, rendererIndex=0}
   {rendererClass=&NiagaraMeshRendererProperties, rendererIndex=1}
   {rendererClass=&NiagaraLightRendererProperties, rendererIndex=2}
  ]
 }
 Minimal003 {
  emitterSpawnScript [/SetVariables(Emitter.Static bool=true, Emitter.bool=true) {isSetParameters=true}]
  emitterUpdateScript [
   &EmitterState(Life Cycle Mode=::NewEnumerator1, Loop Behavior=::NewEnumerator1, Loop Duration Mode=::NewEnumerator1, UseLoopDelay=false)
   &SpawnBurst_Instantaneous(Spawn Count=2, Use Spawn Probability=false, UseLoopCountLimit=false)
   &SpawnRate(SpawnRate=4, Use Spawn Probability=false) {enabled=false}
  ]
  particleSpawnScript [
   &PartitionParticles()
   &InitializeParticle(
    Lifetime=3.75
    Mesh Scale=<0.3 0.3 1>
    Mesh Scale Mode=::NewEnumerator4
    Sprite Size Mode=::NewEnumerator3
    UsePositionOffset=false
    Write Material Random=true
   )
   /SetVariables(Particles.Position=&AddVectorToPosition(
     Position=$Particles.Position
     Vector=&MakeVector(
      X=&Multiply_Float(
       A=70
       B=&Lerp_Float(A=-1, B=1, Alpha=&MakeFloatFromInt(Int=$Particles.PartitionParticles.Partition))
      )
      Y=40
      Z=10
     )
    )) {isSetParameters=true}
   &InitialMeshOrientation(
    Mesh Axis To Orient=<0 0 1>
    Mesh Orientation Mode=::NewEnumerator2
    Orientation Vector=<0 0 -1>
    Rotation=&Multiply_VectorByFloat(
     Vector=<0 0 0>
     Float=&Lerp_Float(A=-1, B=1, Alpha=&MakeFloatFromInt(Int=$Particles.PartitionParticles.Partition))
    )
    Use Rotation Vector=true
   )
   /SetVariables(Particles.AudioPosition:NiagaraFloat=&ReturnNormalizedExecIndex()) {isSetParameters=true}
  ]
  particleUpdateScript [
   &ParticleState(
    Kill Particles When Lifetime Has Elapsed=false
    Let Infinitely Lived Particles Die When Emitter Deactivates=true
    Loop Particles Lifetime=true
   )
   &ScaleMeshSize(Scale Factor=&MakeVector(
     X=1
     Y=1
     Z=&FloatFromCurve(FloatCurve=DataInterface() {propertyValues {Curve {keys [<at 0 0>,<at 0.77 1>]}}}, CurveIndex=$Emitter.Age)
    ))
   &UpdateMeshOrientation(
    Rotation Rate=&Multiply_Float(
     A=1
     B=&Waveform(Global Amplitude Scale=0.5, "Waveform [1]"=::NewEnumerator0, "[1] Phase Scale"=0.3, "[1] Frequency"=1, "[1] Amplitude Min/Max"=<-1.2 1.2>)
    )
    Rotation Vector=&MakeVector(
     X=0
     Y=&Multiply_Float(
      A=0.15
      B=&Lerp_Float(A=-1, B=1, Alpha=&MakeFloatFromInt(Int=$Particles.PartitionParticles.Partition))
     )
     Z=0
    )
   )
   /SetVariables(Particles.GnomonScale:Vector3f=&VectorFromFloat(Value=2)) {isSetParameters=true}
   &RotateAroundPoint(
    Radius=120
    Rotation Rate:NiagaraFloat=&Waveform(
     "Waveform [1]"=::NewEnumerator0
     "[1] Phase Scale"=&RandomRangeFloat(Minimum=0.1, Maximum=0.2)
     "[1] Frequency"=1
     "[1] Amplitude Min/Max"=<-1 1>
     "[1] Scale"=0.07
    )
    X Axis:Vector3f=&Multiply_VectorByFloat(
     Vector=<1 0 0>
     Float=&Lerp_Float(A=-1, B=1, Alpha=&MakeFloatFromInt(Int=$Particles.PartitionParticles.Partition))
    )
   ) {enabled=false, module=RotateAroundPoint001}
   /SetVariables(Particles.Position=&AddVectorToPosition(
     Position=$StackContext.Position
     Vector=&MakeVector(
      X=0
      Y=0
      Z=&Waveform(
       "Waveform [1]"=::NewEnumerator0
       "[1] Frequency"=&RandomRangeFloat(Minimum=0.5, Maximum=1)
       "[1] Amplitude Min/Max"=<-1 1>
       "[1] Scale"=&RandomRangeFloat(Minimum=4, Maximum=6)
      )
     )
    )) {enabled=false, isSetParameters=true}
   &ScaleColor(
    Scale RGB=&Lerp_Vector(
     A=&MakeVectorFromLinearColorRGB(Color=<rgba 0.0078 0.9137 0.8392 1>)
     B=&Multiply_Vector(
      A=&MakeVectorFromLinearColorRGB(Color=&ColorFromCurve(ColorCurve=DataInterface() {propertyValues {
          AlphaCurve {keys [<at 0 1 RCIM_Cubic>]}
          BlueCurve {keys [<at 0 2.7345 RCIM_Cubic>,<at 0.2215 0.9098>,<at 0.25 0>,<at 0.4756 0>,<at 0.5 0.0588>,<at 0.7111 0.0196>,<at 0.75 6>,<at 0.96 1>,<at 1 0.9098>]}
          GreenCurve {keys [<at 0 3 RCIM_Cubic>,<at 0.2215 1>,<at 0.25 3>,<at 0.4756 1>,<at 0.5 3>,<at 0.7111 1>,<at 0.75 0>,<at 0.96 0>,<at 1 1>]}
          RedCurve {keys [<at 0 0.0133 RCIM_Cubic>,<at 0.2215 0.0039>,<at 0.25 2.8353>,<at 0.4756 0.9451>,<at 0.5 0.5529>,<at 0.7111 0.1843>,<at 0.75 3.8824>,<at 0.96 0.65>,<at 1 0.0039>]}
         }}))
      B=&VectorFromFloat(Value=0.65)
     )
     Alpha=&FloatFromCurve(FloatCurve=DataInterface() {propertyValues {Curve {keys [<at 1.7 0>,<at 2 1>]}}}, CurveIndex=$Emitter.Age)
    )
    ScaleA=true
    ScaleRGB=true
   ) {enabled=false, module=ScaleColor002}
   &ScaleColor(
    Scale RGB=&Lerp_Vector(
     A=&MakeVectorFromLinearColorRGB(Color=<rgba 0.0078 0.9137 0.8392 1>)
     B=&Multiply_Vector(
      A=&MakeVectorFromLinearColorRGB(Color=&ColorFromCurve(
        ColorCurve=DataInterface() {propertyValues {
          AlphaCurve {keys [<at 0 1 RCIM_Cubic>]}
          BlueCurve {keys [<at 0 0.8078 RCIM_Cubic>,<at 0.2246 0>,<at 0.3968 0.0196>,<at 0.6058 0.8863>,<at 0.8235 0.8863>,<at 1 0.807 RCIM_Cubic>]}
          GreenCurve {keys [<at 0 0.8863 RCIM_Cubic>,<at 0.2246 1>,<at 0.3968 1>,<at 0.6058 0.3059>,<at 0.8235 0>,<at 1 0.8879 RCIM_Cubic>]}
          RedCurve {keys [<at 0 0.0039 RCIM_Cubic>,<at 0.2246 0.7793>,<at 0.3968 0.1843>,<at 0.6058 0>,<at 0.8235 0.4784>,<at 1 0.0052 RCIM_Cubic>]}
         }}
        CurveIndex=&Multiply_Float(
         A=&Sine(Normalized Angle=&Multiply_Float(A=$Emitter.Age, B=0.1), Period=1, Scale=1, Bias=0)
         B=&Multiply_Float(
          A=&RemapRange(Input Value To Remap=$NPC.FortniteNPC.EventAudio_Amplitude, Input Min=0.65, Input Max=0.88, Output Min=0, Output Max=3)
          B=&Lerp_Float(
           A=0
           B=&MaskFloatBySpawnGroup(Masked Float=1.1, Passthrough Float=1.5)
           Alpha=&RemapRange(Input Value To Remap=$NPC.FortniteNPC.EventAudio_Amplitude, Input Min=0.65, Input Max=0.88, Output Min=0, Output Max=1)
          )
         )
        )
       ))
      B=&VectorFromFloat(Value=0.65)
     )
     Alpha=&FloatFromCurve(FloatCurve=DataInterface() {propertyValues {Curve {keys [<at 1.7 0>,<at 2 1>]}}}, CurveIndex=$Emitter.Age)
    )
    ScaleA=true
    ScaleRGB=true
   ) {enabled=false, module=ScaleColor004}
   &ScaleColor(
    Scale RGB=&Lerp_Vector(
     A=&MakeVectorFromLinearColorRGB(Color=<rgba 0.0078 0.9137 0.8392 1>)
     B=&Lerp_Vector(
      A=&MakeVectorFromLinearColorRGB(Color=<rgba 0.0078 0.9137 0.8392 1>)
      B=&Multiply_Vector(
       A=&MakeVectorFromLinearColorRGB(Color=&ColorFromCurve(
         ColorCurve=DataInterface() {propertyValues {
           AlphaCurve {keys [<at 0 1 RCIM_Cubic>]}
           BlueCurve {keys [<at 0.1161 1>,<at 0.7518 2>]}
           GreenCurve {keys [<at 0.1161 0.9688>,<at 0.7518 0.4613>]}
           RedCurve {keys [<at 0.1161 0.9844>,<at 0.7518 0>]}
          }}
         CurveIndex=&Sine(Normalized Angle=&Add_Float(A=$Emitter.Age, B=1), Period=1, Scale=1, Bias=0)
        ))
       B=&MakeVectorFromLinearColorRGB(Color=<rgba 0.0078 0.9137 0.8392 1>)
      )
      Alpha=&FloatFromCurve(FloatCurve=DataInterface() {propertyValues {Curve {keys [<at 1.7 0>,<at 2 1>]}}}, CurveIndex=$Emitter.Age)
     )
     Alpha=&RemapRange(
      Input Value To Remap=&MakeFloatFromLinearColor(LinearColor=$NPC.NPC_Audio_Reactive_Parameters.Spectrum_Music, Channel:ENiagaraLinearColor_Channels=::NewEnumerator2)
      Input Min=0.32
      Input Max=0.55
      Output Min=0
      Output Max=1
     )
    )
    ScaleA=true
    ScaleRGB=true
   ) {module=ScaleColor005}
   &ScaleColor(
    Scale Alpha=&FloatFromCurve(FloatCurve=DataInterface() {propertyValues {Curve {keys [<at 0.2 0>,<at 0.68 0.4>,<at 0.75 1.2>,<at 1.5 0.4>]}}}, CurveIndex=$Emitter.Age)
    Scale RGB=&VectorFromFloat(Value=&FloatFromCurve(FloatCurve=DataInterface() {propertyValues {Curve {keys [<at 0.68 1>,<at 0.75 1.2>,<at 1.5 1>]}}}, CurveIndex=$Emitter.Age))
    ScaleA=true
    ScaleRGB=true
   )
   &DynamicMaterialParameters(
    Index 0 Param 1=0.15
    Module.Index0Param0Write=true
    Module.Index0Param1Write=true
    Module.Index0Param2Write=true
    Module.Index0Param3Write=true
   )
   /SetVariables(
    Particles.RingPosition:NiagaraPosition=&AddVectorToPosition(Position=$Particles.Position, Vector=<0 0 3>)
    Particles.RingScale:Vector3f=&VectorFromFloat(Value=0.25)
   ) {isSetParameters=true}
   /SetVariables(Particles.DiskScale:Vector3f=&Multiply_VectorByFloat(Vector=<1 1 0.5>, Float=1)) {isSetParameters=true}
   /SetVariables(
    Particles.GlowColor=&MakeLinearColorFromVectorAndFloat(
     "Vector (RGB)"=&Lerp_Vector(
      A=&MakeVectorFromLinearColorRGB(Color=<rgba 0.0078 0.9137 0.8392 1>)
      B=&Lerp_Vector(
       A=&MakeVectorFromLinearColorRGB(Color=<rgba 0.0078 0.9137 0.8392 1>)
       B=&Multiply_Vector(
        A=&MakeVectorFromLinearColorRGB(Color=&ColorFromCurve(
          ColorCurve=DataInterface() {propertyValues {
            AlphaCurve {keys [<at 0 1 RCIM_Cubic>]}
            BlueCurve {keys [<at 0.1161 1>,<at 0.7518 2>]}
            GreenCurve {keys [<at 0.1161 0.9688>,<at 0.7518 0.4613>]}
            RedCurve {keys [<at 0.1161 0.9844>,<at 0.7518 0>]}
           }}
          CurveIndex=&Sine(Normalized Angle=&Add_Float(A=$Emitter.Age, B=1), Period=1, Scale=1, Bias=0)
         ))
        B=&MakeVectorFromLinearColorRGB(Color=<rgba 0.0078 0.9137 0.8392 1>)
       )
       Alpha=&FloatFromCurve(FloatCurve=DataInterface() {propertyValues {Curve {keys [<at 1.7 0>,<at 2 1>]}}}, CurveIndex=$Emitter.Age)
      )
      Alpha=&RemapRange(
       Input Value To Remap=&MakeFloatFromLinearColor(LinearColor=$NPC.NPC_Audio_Reactive_Parameters.Spectrum_Music, Channel:ENiagaraLinearColor_Channels=::NewEnumerator2)
       Input Min=0.29
       Input Max=0.5
       Output Min=0
       Output Max=1
      )
     )
     "Float (Alpha)"=&FloatFromCurve(FloatCurve=DataInterface() {propertyValues {Curve {keys [<at 0.2 0>,<at 0.68 0.2>,<at 0.75 0.6>,<at 1.5 0.2>]}}}, CurveIndex=$Emitter.Age)
    )
    Particles.GlowScale:Vector2f=&Vector2DFromFloat(Value=150)
   ) {isSetParameters=true}
   &ScaleColor(
    Color Value To Scale=$Particles.GlowColor
    Scale RGB=&Multiply_Vector(
     A=&MakeVectorFromLinearColorRGB(Color=&ColorFromCurve(ColorCurve=DataInterface() {propertyValues {
         AlphaCurve {keys [<at 0 5 RCIM_Cubic>]}
         BlueCurve {keys [<at 0 0.807 RCIM_Cubic>,<at 0.2246 0>,<at 0.3968 0.019>,<at 0.6058 0.8879>,<at 1 0.8879>]}
         GreenCurve {keys [<at 0 0.8879 RCIM_Cubic>,<at 0.2246 1>,<at 0.3968 1>,<at 0.6058 0.3043>,<at 1 0>]}
         RedCurve {keys [<at 0 0.0052 RCIM_Cubic>,<at 0.2246 0.9451>,<at 0.3968 0.1838>,<at 0.6058 0>,<at 1 0.4788>]}
        }}))
     B=&VectorFromFloat(Value=0.65)
    )
    ScaleA=true
    ScaleRGB=true
   ) {enabled=false, module=ScaleColor003}
   &ScaleColor(
    Scale RGB=&Multiply_Vector(
     A=&MakeVectorFromLinearColorRGB(Color=&ColorFromCurve(
       ColorCurve=DataInterface() {propertyValues {
         AlphaCurve {keys [<at 0 1 RCIM_Cubic>]}
         BlueCurve {keys [<at 0 0.8078 RCIM_Cubic>,<at 0.2246 0>,<at 0.3968 0.0196>,<at 0.6058 0.8863>,<at 0.8235 0.8863>,<at 1 0.807 RCIM_Cubic>]}
         GreenCurve {keys [<at 0 0.8863 RCIM_Cubic>,<at 0.2246 1>,<at 0.3968 1>,<at 0.6058 0.3059>,<at 0.8235 0>,<at 1 0.8879 RCIM_Cubic>]}
         RedCurve {keys [<at 0 0.0039 RCIM_Cubic>,<at 0.2246 0.7793>,<at 0.3968 0.1843>,<at 0.6058 0>,<at 0.8235 0.4784>,<at 1 0.0052 RCIM_Cubic>]}
        }}
       CurveIndex=&Multiply_Float(
        A=&Sine(Normalized Angle=&Add_Float(A=$Emitter.Age, B=0), Period=1, Scale=0.1, Bias=0)
        B=&Multiply_Float(
         A=&RemapRange(Input Value To Remap=$NPC.FortniteNPC.EventAudio_Amplitude, Input Min=0.65, Input Max=0.88, Output Min=0, Output Max=3)
         B=&Lerp_Float(
          A=0
          B=&MaskFloatBySpawnGroup(Masked Float=1.1, Passthrough Float=1.5)
          Alpha=&RemapRange(Input Value To Remap=$NPC.FortniteNPC.EventAudio_Amplitude, Input Min=0.65, Input Max=0.88, Output Min=0, Output Max=1)
         )
        )
       )
      ))
     B=&VectorFromFloat(Value=0.65)
    )
    ScaleA=true
    ScaleRGB=true
   ) {enabled=false, module=ScaleColor001}
   &Color() {enabled=false}
  ]
  rendererClasses [&NiagaraSpriteRendererProperties, &NiagaraMeshRendererProperties]
  renderers [
   {rendererClass=&NiagaraSpriteRendererProperties}
   {rendererClass=&NiagaraMeshRendererProperties, rendererIndex=1}
   {rendererClass=&NiagaraMeshRendererProperties, rendererIndex=2}
   {rendererClass=&NiagaraMeshRendererProperties, rendererIndex=3}
   {rendererClass=&NiagaraSpriteRendererProperties, rendererIndex=4}
  ]
 }
}
summary {
 systemName=NS_JadeTowelLights
 emitters [
  {emitterName=Minimal002, rendererClasses [&NiagaraMeshRendererProperties, &NiagaraLightRendererProperties]}
  {emitterName=Minimal003, rendererClasses [&NiagaraSpriteRendererProperties, &NiagaraMeshRendererProperties]}
 ]
 userVariables [
  User.bIsFrontEnd {type=NiagaraBool, defaultValue=true}
  User.bIsFrontEndPreview {type=NiagaraBool}
  User.Linear Color {type=LinearColor, description="", defaultValue {struct=LinearColor, value=<rgba 0.0078 0.9137 0.8392 1>}}
  User.Local Alpha Preview {type=NiagaraFloat, defaultValue=0.4, description=""}
  User.Other Alpha Preview {type=NiagaraFloat, defaultValue=0.4, description=""}
 ]
}
