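# Niagara system for the JadeTowel emote: a music-reactive neon ring effect with an intro burst, glow and falling debris.
# Short-lived Burst_Ray and Glow_Body meshes accompany an Inner_Ring, an Outer_Ring of 28 meshes tinted by User.SquareRingColor, and timed bursts of colored Pulsing_Squares sprites on the same rotating circle.
# Event audio amplitude and spectrum drive ring scale and material parameters, while Falling_Tris drift down with amplitude-driven material values.
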NiagaraSystem /BRCosmetics/Animation/Game/MainPlayer/Emotes/JadeTowel/FX/NS_JadeTowelRing
dependencies {
 usedDataInterfaces [
  &NiagaraDataInterfaceArrayColor
  &NiagaraDataInterfaceCurve
  &NiagaraDataInterfaceColorCurve
  &NiagaraDataInterfaceVectorCurve
 ]
 usedDynamicInputs [&ConvertVectorToPosition, &Waveform, &InvertBool]
 usedModules [
  &SystemState
  &InterpolateOverTime
  &self:SystemSpawnScript.SystemScriptSource.SystemScriptGraph.NiagaraNodeAssignment_0.SetVariables_#1
  &self:SystemSpawnScript.SystemScriptSource.SystemScriptGraph.NiagaraNodeAssignment_1.SetVariables_#2
  &self:SystemSpawnScript.SystemScriptSource.SystemScriptGraph.NiagaraNodeAssignment_2.SetVariables_#3
  &self:SystemSpawnScript.SystemScriptSource.SystemScriptGraph.NiagaraNodeAssignment_3.SetVariables_#4
  &self:SystemSpawnScript.SystemScriptSource.SystemScriptGraph.NiagaraNodeAssignment_4.SetVariables_#5
 ]
}
emitters {
 Burst_Ray {
  emitterUpdateScript [
   &EmitterState(Life Cycle Mode=::NewEnumerator1, Loop Behavior=::NewEnumerator1, Loop Duration=0.4, UseLoopDelay=true)
   &SpawnRate(
    SpawnRate=&FloatFromCurve(FloatCurve=DataInterface() {propertyValues {Curve {keys [<at 0 0.6>,<at 1 1>]}}}, CurveIndex=$Emitter.Age, Scale Curve=100)
    Use Spawn Probability=false
   )
   &SpawnBurst_Instantaneous(Spawn Count=5, Use Spawn Probability=false, UseLoopCountLimit=false)
  ]
  particleSpawnScript [
   &InitializeParticle(
    Color=<rgba 0.1508 0.5 0.399 1>
    Lifetime Max=0.5
    Lifetime Min=0.15
    Lifetime Mode=::NewEnumerator1
    Mesh Scale=&Multiply_VectorByFloat(Vector=&RandomRangeVector(Minimum=<0.2 0.2 1>, Maximum=<0.4 0.4 1>), Float=0.1)
    Mesh Scale Mode=::NewEnumerator4
    UsePositionOffset=false
    Write Material Random=true
   )
   &ShapeLocation(
    Handle Radius=10
    Large Radius=&FloatFromCurve(
     FloatCurve=DataInterface() {propertyValues {Curve {keys [
         {arriveTangent=1.8769, interpMode=RCIM_Cubic, leaveTangent=1.8769, tangentMode=RCTM_User, time=0, value=0.2}
         {arriveTangent=0.8, interpMode=RCIM_Cubic, leaveTangent=0.8, time=0.3519, value=0.7183}
         {arriveTangent=0.0876, interpMode=RCIM_Cubic, leaveTangent=0.0876, tangentMode=RCTM_User, time=1, value=1}
        ]}}}
     CurveIndex=$Emitter.Age
     Scale Curve=150
    )
    MaskSpawnGroup=false
    Override Seed=false
    Shape Primitive=::NewEnumerator3
    UseOwnerScale=true
   ) {module=ShapeLocation001}
  ]
  particleUpdateScript [
   &ParticleState()
   &ScaleMeshSize(Scale Factor=&VectorFromCurve(VectorCurve=DataInterface() {propertyValues {
       XCurve {keys [<at 0 1 RCIM_Cubic>,<at 1 1>]}
       YCurve {keys [<at 0 1 RCIM_Cubic>,<at 1 1>]}
       ZCurve {keys [
         {arriveTangent=1.4897, interpMode=RCIM_Cubic, leaveTangent=1.4897, tangentMode=RCTM_User, time=0, value=0.3}
         {arriveTangent=0.2789, tangentMode=RCTM_User, time=1, value=1.5}
        ]}
      }}))
   &DynamicMaterialParameters(
    Index 0 Param 1=1
    Index 0 Param 2=1
    Index 0 Param 3=1
    Index 0 Param 4=1
    Module.Index0Param0Write=true
    Module.Index0Param1Write=true
    Module.Index0Param2Write=true
    Module.Index0Param3Write=true
   )
   &CameraOffset(Camera Offset Amount=15)
   &ScaleColor(
    Scale Alpha=&FloatFromCurve(FloatCurve=DataInterface() {propertyValues {Curve {keys [{arriveTangent=0.0239, interpMode=RCIM_Cubic, leaveTangent=0.0239, tangentMode=RCTM_User, time=0, value=1}, <at 1 0>]}}})
    Scale RGB=&Multiply_Vector(
     A=&VectorFromFloat(Value=8)
     B=&MakeVectorFromLinearColorRGB(Color=&ColorFromCurve(ColorCurve=DataInterface() {propertyValues {
         AlphaCurve {keys [<at 0 1 RCIM_Cubic>]}
         BlueCurve {keys [<at 0.0176 0.7275>,<at 0.1843 0.5265>,<at 0.4216 1>,<at 0.548 1>,<at 0.7273 0.1>,<at 1 0.599 RCIM_Cubic>]}
         GreenCurve {keys [<at 0.0176 1>,<at 0.1843 1>,<at 0.4216 0.5705>,<at 0.548 0>,<at 0.7273 1>,<at 1 1 RCIM_Cubic>]}
         RedCurve {keys [<at 0.0176 0>,<at 0.1843 0>,<at 0.4216 0>,<at 0.548 0.3031>,<at 0.7273 0>,<at 1 0 RCIM_Cubic>]}
        }}))
    )
    ScaleA=true
    ScaleRGB=true
   )
   &FN_Mesh_Attributes(
    Color=&MakeLinearColorFromVectorAndFloat(
     "Vector (RGB)"=&Multiply_VectorByFloat(
      Vector=&MakeVectorFromLinearColorRGB(Color=&ColorFromCurve(ColorCurve=DataInterface() {propertyValues {
          AlphaCurve {keys [<at 0 1 RCIM_Cubic>]}
          BlueCurve {keys [<at 0.0176 0.7275>,<at 0.1843 0.5265>,<at 0.4216 1>,<at 0.548 1>,<at 0.7273 0.1>,<at 1 0.599 RCIM_Cubic>]}
          GreenCurve {keys [<at 0.0176 1>,<at 0.1843 1>,<at 0.4216 0.5705>,<at 0.548 0>,<at 0.7273 1>,<at 1 1 RCIM_Cubic>]}
          RedCurve {keys [<at 0.0176 0>,<at 0.1843 0>,<at 0.4216 0>,<at 0.548 0.3031>,<at 0.7273 0>,<at 1 0 RCIM_Cubic>]}
         }}))
      Float=10
     )
     "Float (Alpha)"=&FloatFromCurve(
      FloatCurve=DataInterface() {propertyValues {Curve {keys [{arriveTangent=0.0703, interpMode=RCIM_Cubic, leaveTangent=0.0703, tangentMode=RCTM_User, time=0, value=1}, <at 1 0>]}}}
      Scale Curve=10
     )
    )
    Create Color Attribute=true
    Create Scale Attribute=true
    Scale=<0.8 0.8 0.1>
    Write to Module Namespace=true
   )
  ]
  rendererClasses [&NiagaraMeshRendererProperties]
  renderers [
   {rendererClass=&NiagaraMeshRendererProperties, rendererIndex=0}
   {rendererClass=&NiagaraMeshRendererProperties, rendererIndex=1}
  ]
 }
 Falling_Tris {
  emitterUpdateScript [
   &EmitterState(Life Cycle Mode=::NewEnumerator1, UseLoopDelay=false)
   &SpawnRate(SpawnRate=15, Use Spawn Probability=false)
  ]
  particleSpawnScript [
   &InitializeParticle(
    Color=&SelectLinearColorFromArray(
     Array Sampling Mode=::NewEnumerator1
     Direct Array Index=&RandomRangeInt(Minimum=0, Maximum=2)
     Color Selection Array=DataInterface() {propertyValues {ColorData [<rgba 0.1394 0.2649 2 1>, <rgba 0 1.6981 2 1>, <rgba 2 0 1.339 1>]}}
     Random Seed=?
    )
    Lifetime Max=6
    Lifetime Min=3
    Lifetime Mode=::NewEnumerator1
    Mass Max=1.5
    Mass Mode=::NewEnumerator3
    Mesh Scale Mode=::NewEnumerator1
    Mesh Uniform Scale Max=0.5
    Mesh Uniform Scale Min=0.3
    Sprite Rotation Mode=::NewEnumerator2
    Sprite Size Mode=::NewEnumerator3
    UsePositionOffset=false
    Write Material Random=true
   )
   &ShapeLocation(
    MaskSpawnGroup=false
    Non Uniform Scale=<0.4228 0.29 0.8374>
    Offset=<0 0 90>
    Offset Mode=::NewEnumerator0
    Override Seed=false
    Sphere Radius=192.2285
    UseOwnerScale=true
   )
   &InitialMeshOrientation(Use Rotation Vector=false)
   &MeshRotationForce(Rotation=&RandomRangeVector(Minimum=<1 1 1>, Maximum=<3 3 3>))
   &ApplyInitialForces()
  ]
  particleUpdateScript [
   &ParticleState()
   &ScaleMeshSize(Scale Factor=&VectorFromFloat(Value=&FloatFromCurve(FloatCurve=DataInterface() {propertyValues {Curve {keys [
          {interpMode=RCIM_Cubic, tangentMode=RCTM_User, time=0, value=0}
          {interpMode=RCIM_Cubic, tangentMode=RCTM_User, time=0.5, value=1}
          {tangentMode=RCTM_User, time=1, value=0}
         ]}}})))
   &GravityForce(Gravity=<0 0 -20>)
   &SolveForcesAndVelocity(Clamp Velocity=false, Limit Acceleration=false)
   &DynamicMaterialParameters(
    Index 0 Param 2=5
    Index 0 Param 3=0.4
    Index 0 Param 4=&RandomRangeFloat(Minimum=0, Maximum=0.2)
    Index 1 Param 1=0.2
    Index 1 Param 2=0.2
    Index 1 Param 4=1
    Index 2 Param 1=1
    Index 3 Param 1=1
    Index 3 Param 2=&Add_Float(
     A=50
     B=&RemapRange(Input Value To Remap=$NPC.FortniteNPC.EventAudio_Amplitude, Input Min=0.6, Input Max=1, Output Min=0, Output Max=500)
    )
    Index 3 Param 3=50
    Module.Index0Param0Write=true
    Module.Index0Param1Write=true
    Module.Index0Param2Write=true
    Module.Index0Param3Write=true
    Module.Index1Param0Write=true
    Module.Index1Param1Write=true
    Module.Index1Param2Write=true
    Module.Index1Param3Write=true
    Module.Index2Param0Write=true
    Module.Index2Param1Write=true
    Module.Index2Param2Write=true
    Module.Index2Param3Write=true
    Module.Index3Param0Write=true
    Module.Index3Param1Write=true
    Module.Index3Param2Write=true
    Module.Index3Param3Write=true
    Write Parameter Index 1=true
    Write Parameter Index 2=true
    Write Parameter Index 3=true
   ) {enabled=false}
   &DynamicMaterialParameters(
    Index 0 Param 1=&Add_Float(
     A=20
     B=&Lerp_Float(
      A=&Multiply_Float(A=$NPC.FortniteNPC.EventAudio_Amplitude, B=$NPC.FortniteNPC.EventAudio_Amplitude)
      B=1
      Alpha=$System.MusicReactivePreview
     )
    )
    Index 0 Param 2=5
    Index 0 Param 3=&RandomRangeFloat(Minimum=0, Maximum=0.2)
    Index 0 Param 4=&RandomRangeFloat(Minimum=0, Maximum=0.2)
    Module.Index0Param0Write=true
    Module.Index0Param1Write=true
    Module.Index0Param2Write=true
    Module.Index0Param3Write=true
   ) {module=DynamicMaterialParameters001}
   &KillParticlesInVolume(Box Size:Vector3f=<300 300 100>, Kill Shape=::NewEnumerator1, Origin Offset:Vector3f=<0 0 -50>)
   &KillParticlesInVolume(Box Size:Vector3f=<50 197.0569 359.1298>, Kill Shape=::NewEnumerator1, Origin Offset:Vector3f=<0 0 50>) {module=KillParticlesInVolume001}
   &ScaleColor(Scale Alpha=&RandomRangeFloat(Minimum=0.85, Maximum=1), ScaleA=true, ScaleRGB=true)
   /SetVariables(
    Particles.Linear Color=/Niagara/DynamicInputs/Add/Add_LinearColor(A:LinearColor=$Particles.Color, B:LinearColor=<rgba 0 0 0 -0.8>)
    Particles.SpriteSize=&Multiply_Vector2D_ByFloat(
     Float=20
     Vector2D=&MakeVector2D(
      X=&MakeFloatFromVector(Vector=$Particles.Scale, Channel=::NewEnumerator0)
      Y=&MakeFloatFromVector(Vector=$Particles.Scale, Channel=::NewEnumerator1)
     )
    )
   ) {isSetParameters=true}
   &ScaleColor(
    Scale Alpha=&FloatFromCurve(FloatCurve=DataInterface() {propertyValues {Curve {keys [
         {interpMode=RCIM_Cubic, tangentMode=RCTM_User, time=0, value=0}
         {interpMode=RCIM_Cubic, tangentMode=RCTM_User, time=0.5, value=1}
         {tangentMode=RCTM_User, time=1, value=0}
        ]}}})
    ScaleA=true
    ScaleRGB=true
   ) {module=ScaleColor001}
   /SetVariables(Particles.Vector 4:Vector4f=&MakeVector4(X=0, Y=0, Z=0, W=0)) {isSetParameters=true}
  ]
  rendererClasses [&NiagaraMeshRendererProperties]
  renderers [{rendererClass=&NiagaraMeshRendererProperties, rendererIndex=0}]
 }
 Glow_Body {
  emitterUpdateScript [
   /SetVariables(Emitter.GlowOutlineThickness:NiagaraFloat=&FloatFromCurve(FloatCurve=DataInterface() {propertyValues {Curve {keys [<at 0 0>,<at 0.6 0.3>,<at 0.7 1>,<at 2 0>]}}}, CurveIndex=$Emitter.Age)) {isSetParameters=true}
   &EmitterState(Life Cycle Mode=::NewEnumerator1, Loop Behavior=::NewEnumerator1, Loop Duration Mode=::NewEnumerator1, UseLoopDelay=false)
   &SpawnBurst_Instantaneous(Use Spawn Probability=false, UseLoopCountLimit=false)
  ]
  particleSpawnScript [&InitializeParticle(
    Lifetime=2.7
    Mesh Scale=<1.5 1.5 2.5>
    Mesh Scale Mode=::NewEnumerator4
    Position Offset=<0 0 60>
    UsePositionOffset=true
    Write Material Random=true
   )]
  particleUpdateScript [
   &ParticleState()
   &CameraOffset(Camera Offset Amount=&Multiply_Float(A=50, B=$System.FN_Contrail_Setup.Dive Value))
   &ScaleColor(
    Scale Alpha=&FloatFromCurve(FloatCurve=DataInterface() {propertyValues {Curve {keys [<at 0 0>,<at 1 1>]}}}, CurveIndex=$System.Timeline.Time)
    ScaleA=true
    ScaleRGB=true
   ) {enabled=false}
   &ScaleColor(
    Scale Alpha=&FloatFromCurve(
     FloatCurve=DataInterface() {propertyValues {Curve {keys [<at 0 0>,<at 0.6 0.3>,<at 0.7 1.5>,<at 2 0>]}}}
     CurveIndex=$Emitter.Age
     Scale Curve=44
    )
    ScaleA=true
    ScaleRGB=true
   ) {module=ScaleColor001}
  ]
  rendererClasses [&NiagaraMeshRendererProperties]
  renderers [
   {rendererClass=&NiagaraMeshRendererProperties, rendererIndex=0}
   {rendererClass=&NiagaraMeshRendererProperties, rendererIndex=1}
  ]
 }
 Inner_Ring {
  emitterUpdateScript [
   &EmitterState(Life Cycle Mode=::NewEnumerator1, Loop Behavior=::NewEnumerator1, Loop Duration=1, UseLoopDelay=false)
   &SpawnBurst_Instantaneous(Spawn Group=1, Use Spawn Probability=false, UseLoopCountLimit=false) {module=SpawnBurst_Instantaneous001}
   &SpawnBurst_Instantaneous(Spawn Count=0, Use Spawn Probability=false, UseLoopCountLimit=false)
  ]
  particleSpawnScript [
   &InitializeParticle(
    Color=<rgba 0 0.679 0.56 1>
    Lifetime=10.37
    Mesh Scale Mode=::NewEnumerator3
    Mesh Uniform Scale=&MaskFloatBySpawnGroup(Masked Float=0.292, Passthrough Float=0.612)
    Position Offset=&MaskVectorBySpawnGroup(Masked Vector=<2.7518 -49.9378 101.004>, Passthrough Vector=<1.1619 -33.1442 100.6603>)
    UsePositionOffset=true
    Write Material Random=true
   )
   &InitialMeshOrientation(
    Mesh Orientation Mode=::NewEnumerator1
    Rotation=&MakeCustomVectorFromBool(
     Bool=$User.bIsFrontEnd
     True Vector=<-0.0553 0 1>
     False Vector=&MaskVectorBySpawnGroup(Masked Vector=<0.1768 0.0244 0.5>, Passthrough Vector=<0.1768 0.0244 0.5>)
    )
    Rotation Coordinate Space=::NewEnumerator1
    Use Rotation Vector=true
   )
   &MeshRotationForce(
    Coordinate Space:ENiagaraMeshTransforms=::NewEnumerator3
    Rotation=&MaskVectorBySpawnGroup(Masked Vector=<0 0.1 0>, Passthrough Vector=<0 -0.1 0>)
   )
   &ApplyInitialForces()
  ]
  particleUpdateScript [
   &DynamicMaterialParameters(
    Index 0 Param 1=&Multiply_Float(
     A=&RemapRange(Input Value To Remap=$NPC.FortniteNPC.EventAudio_Amplitude, Input Min=0.7, Input Max=0.85, Output Min=0, Output Max=1)
     B=&Lerp_Float(
      A=0
      B=&MaskFloatBySpawnGroup(Masked Float=1.1, Passthrough Float=1.5)
      Alpha=&RemapRange(Input Value To Remap=$NPC.FortniteNPC.EventAudio_Amplitude, Input Min=0.65, Input Max=0.88, Output Min=0, Output Max=1)
     )
    )
    Index 0 Param 2=&RandomRangeFloat(Minimum=0.1, Maximum=0.4)
    Index 0 Param 3=&RandomRangeFloat(Minimum=-0.1, Maximum=-0.3)
    Index 0 Param 4=0.2
    Index 1 Param 1=1
    Index 1 Param 2=2
    Index 1 Param 3=0.5
    Index 1 Param 4=1
    Index 2 Param 1=1
    Index 3 Param 1=&FloatFromCurve(FloatCurve=DataInterface() {propertyValues {Curve {keys [<at 0 -0.4>,<at 0.6 0.7>]}}}, CurveIndex=$Emitter.Age)
    Module.Index0Param0Write=true
    Module.Index0Param1Write=true
    Module.Index0Param2Write=true
    Module.Index0Param3Write=true
    Module.Index1Param0Write=true
    Module.Index1Param1Write=true
    Module.Index1Param2Write=true
    Module.Index1Param3Write=true
    Module.Index2Param0Write=true
    Module.Index2Param1Write=true
    Module.Index2Param2Write=true
    Module.Index2Param3Write=true
    Module.Index3Param0Write=true
    Module.Index3Param1Write=true
    Module.Index3Param2Write=true
    Module.Index3Param3Write=true
    Write Parameter Index 1=true
    Write Parameter Index 2=true
    Write Parameter Index 3=true
   ) {module=DynamicMaterialParameters001}
   &SolveForcesAndVelocity(Clamp Velocity=false, Limit Acceleration=false)
   &ParticleState(Kill Particles When Lifetime Has Elapsed=false)
   &ScaleColor(
    Scale RGB=&VectorFromFloat(Value=&Lerp_Float(
      A=1
      B=0.2
      Alpha=&FNTimeOfDayScalar(Night Scale=1, Morning Scale=0.5, Day Scale=0, Evening Scale=0.5)
     ))
    ScaleA=true
    ScaleRGB=true
   ) {enabled=false, module=ScaleColor001}
   &Color(Color=&ColorFromCurve(ColorCurve=DataInterface() {propertyValues {
       AlphaCurve {keys [<at 0.0024 0>, <at 0.0776 0.5>, {arriveTangent=-0.5421, interpMode=RCIM_Cubic, leaveTangent=-0.5421, time=0.9166, value=0.5}, <at 1 0>]}
       BlueCurve {keys [{arriveTangent=5.7095, interpMode=RCIM_Cubic, time=0.25, value=3.325}, <at 0.35 4>, <at 0.45 3.8408>, <at 0.5 4>, <at 0.6 4>, <at 0.7 4>]}
       GreenCurve {keys [
         {arriveTangent=-5.4702, interpMode=RCIM_Cubic, time=0.25, value=4}
         <at 0.35 3.3533>
         <at 0.45 4>
         <at 0.5 3.0171>
         <at 0.6 3.52>
         <at 0.7 3.6812>
        ]}
       RedCurve {keys [<at 0.25 0 RCIM_Cubic>,<at 0.35 0>,<at 0.45 0.7041>,<at 0.5 0>,<at 0.6 1.1639>,<at 0.7 0.5671>]}
      }})) {enabled=false, module=Color001}
   &ScaleColor(ScaleA=true, ScaleRGB=true) {enabled=false}
   &DynamicMaterialParameters(
    Index 0 Param 1=&Add_Float(
     A=&Sine(Normalized Angle=$Emitter.Age, Period=&RandomRangeFloat(Minimum=0.3, Maximum=1.7), Scale=1, Bias=0)
     B=&Lerp_Float(
      A=&Multiply_Float(A=$NPC.FortniteNPC.EventAudio_Amplitude, B=$NPC.FortniteNPC.EventAudio_Amplitude)
      B=1
      Alpha=$System.MusicReactivePreview
     )
    )
    Index 0 Param 2=&RandomRangeFloat(Minimum=0.2, Maximum=0.5)
    Index 0 Param 3=&RandomRangeFloat(Minimum=-0.2, Maximum=-0.5)
    Index 0 Param 4=0.2
    Index 1 Param 1=1
    Index 1 Param 2=2
    Index 1 Param 3=0.2
    Index 2 Param 1=1
    Module.Index0Param0Write=true
    Module.Index0Param1Write=true
    Module.Index0Param2Write=true
    Module.Index0Param3Write=true
    Module.Index1Param0Write=true
    Module.Index1Param1Write=true
    Module.Index1Param2Write=true
    Module.Index1Param3Write=true
    Module.Index2Param0Write=true
    Module.Index2Param1Write=true
    Module.Index2Param2Write=true
    Module.Index2Param3Write=true
    Write Parameter Index 1=true
    Write Parameter Index 2=true
   ) {enabled=false}
  ]
  rendererClasses [&NiagaraMeshRendererProperties]
  renderers [{rendererClass=&NiagaraMeshRendererProperties, rendererIndex=0}]
 }
 Outer_Ring {
  emitterUpdateScript [
   &EmitterState(Life Cycle Mode=::NewEnumerator1, Loop Behavior=::NewEnumerator1, Loop Duration Mode=::NewEnumerator1, UseLoopDelay=false)
   &SpawnBurst_Instantaneous(Spawn Count=28, Use Spawn Probability=false, UseLoopCountLimit=false)
  ]
  particleSpawnScript [
   &InitializeParticle(
    Color=&LerpLinearColors(
     StartColor=$User.SquareRingColor
     EndColor=&LerpLinearColors(
      StartColor=<rgba 0 1 0.3741 1>
      EndColor=$User.SquareRingColor
      LerpFactor=&Lerp_Float(
       A=0
       B=1
       Alpha=&RemapRange(Input Value To Remap=&ReturnNormalizedExecIndex(), Input Min=0.5, Input Max=1, Output Min=0, Output Max=1)
      )
     )
     LerpFactor=&ReturnNormalizedExecIndex()
    )
    Lifetime=0.9375
    Mesh Scale Mode=::NewEnumerator3
    Mesh Uniform Scale=0.06
    Sprite Rotation Angle=90
    Sprite Rotation Mode=::NewEnumerator0
    UsePositionOffset=false
    Write Material Random=true
   )
   &MeshRotationForce(Coordinate Space:ENiagaraMeshTransforms=::NewEnumerator2, Rotation=<0 0.8 0>) {enabled=false}
   &ApplyInitialForces()
   &AddVelocity(Module.Use Falloff Distance=false, Override Seed=false, Velocity Mode=::NewEnumerator1, Velocity Speed=0.1)
   &SpriteFacingAndAlignment(Sprite Alignment=$Particles.ShapeLocation.ShapeBitangent, Sprite Facing=<0.01 1 0>, WriteAlignment=true, WriteFacing=true) {enabled=false}
  ]
  particleUpdateScript [
   &ParticleState(
    Kill Particles When Lifetime Has Elapsed=false
    Let Infinitely Lived Particles Die When Emitter Deactivates=true
    Loop Particles Lifetime=true
   )
   &ScaleMeshSize(Scale Factor=&VectorFromFloat(Value=&Lerp_Float(
      A=&FloatFromCurve(FloatCurve=DataInterface() {propertyValues {Curve {keys [<at 0 0>,<at 0.7 0.5>]}}}, CurveIndex=$Emitter.Age)
      B=&Multiply_Float(
       A=&RemapRange(
        Input Value To Remap=&MakeFloatFromLinearColor(LinearColor=$NPC.FortniteNPC.EventAudio_Spectrum, Channel:ENiagaraLinearColor_Channels=::NewEnumerator2)
        Input Min=0.65
        Input Max=0.88
        Output Min=0.65
        Output Max=1
       )
       B=&Lerp_Float(
        A=0
        B=&MaskFloatBySpawnGroup(Masked Float=1.1, Passthrough Float=1.5)
        Alpha=&RemapRange(Input Value To Remap=$NPC.FortniteNPC.EventAudio_Amplitude, Input Min=0.65, Input Max=0.88, Output Min=0.65, Output Max=1)
       )
      )
      Alpha=&FloatFromCurve(FloatCurve=DataInterface() {propertyValues {Curve {keys [<at 0.68 0>,<at 0.75 0>,<at 0.8 1>]}}}, CurveIndex=$Emitter.Age)
     )))
   &ShapeLocation(
    MaskSpawnGroup=false
    Offset=<1 -22 101>
    Offset Mode=::NewEnumerator0
    Override Seed=false
    Ring / Disc Distribution Mode:ENiagaraSphereDistributionMode=::NewEnumerator1
    Ring Radius=68
    Rotation Angle:NiagaraFloat=&Multiply_Float(A=0.016, B=$Engine.System.Age)
    Rotation Axis=&MakeVector(X=0, Y=1, Z=0)
    Rotation Mode=::NewEnumerator1
    Shape Primitive=::NewEnumerator4
    U Position=&ReturnNormalizedExecIndex()
    UseOwnerScale=true
   )
   &UpdateMeshOrientation(
    Facing Direction=$Particles.ShapeLocation.ShapeBitangent
    Facing Interpolation=false
    Orientation Method=::NewEnumerator1
    Override Quat=false
    Reference Axis Vector:Vector3f=<0 0 1>
    Stabilization Mode:ENiagara_OrientVectorCount=::NewEnumerator0
    TransformFacing=false
   )
   &SolveForcesAndVelocity(Clamp Velocity=false, Limit Acceleration=false)
   &GenerateLocationEvent(Use Event Probability=false, UseEventDelay=false, "Vector to Send as Vector 2 (Velocity)":Vector3f=$Particles.ShapeLocation.ShapeBitangent)
   &AlignSpriteToMeshOrientation(
    Mesh Orientation Relative Sprite Alignment Vector=$Particles.ShapeLocation.ShapeTangent
    Mesh Orientation Relative Sprite Facing Vector:Vector3f=<0.1 1 0>
    SetAlignment=true
    SetFacing=true
   ) {enabled=false}
   &DynamicMaterialParameters(
    Index 0 Param 2=&RandomRangeFloat(Minimum=0.2, Maximum=0.5)
    Index 0 Param 3=&RandomRangeFloat(Minimum=-0.2, Maximum=-0.5)
    Index 0 Param 4=0.2
    Index 1 Param 1=1
    Index 1 Param 2=2
    Index 1 Param 3=0.5
    Index 1 Param 4=0.5
    Index 2 Param 1=1
    Index 3 Param 1=&FloatFromCurve(FloatCurve=DataInterface() {propertyValues {Curve {keys [<at 0 -0.4>,<at 0.7 0.6>]}}}, CurveIndex=$Emitter.Age)
    Module.Index0Param0Write=true
    Module.Index0Param1Write=true
    Module.Index0Param2Write=true
    Module.Index0Param3Write=true
    Module.Index1Param0Write=true
    Module.Index1Param1Write=true
    Module.Index1Param2Write=true
    Module.Index1Param3Write=true
    Module.Index2Param0Write=true
    Module.Index2Param1Write=true
    Module.Index2Param2Write=true
    Module.Index2Param3Write=true
    Module.Index3Param0Write=true
    Module.Index3Param1Write=true
    Module.Index3Param2Write=true
    Module.Index3Param3Write=true
    Write Parameter Index 1=true
    Write Parameter Index 2=true
    Write Parameter Index 3=true
   ) {enabled=false}
   &self:NoiseScale_01(
    Noise Frequency=7
    Noise Speed:NiagaraFloat=&Lerp_Float(A=1.5, B=3.5, Alpha=&Ceil_Float(Float=0))
    Scale Size:NiagaraFloat=&Lerp_Float(A=0, B=1, Alpha=1)
   ) {module=NoiseScale_01}
   &ScaleMeshSize(Scale Factor=&MakeVector(
     X=&Lerp_Float(
      A=1
      B=&Multiply_Float(
       A=&RemapRange(
        Input Value To Remap=/Niagara/DynamicInputs/Math/LerpMultipleFloats(
         Float Count:ENiagara_MultipleLerpCount=::NewEnumerator1
         1:NiagaraFloat=&MakeFloatFromLinearColor(LinearColor=$NPC.FortniteNPC.EventAudio_Spectrum, Channel:ENiagaraLinearColor_Channels=::NewEnumerator0)
         2:NiagaraFloat=&MakeFloatFromLinearColor(LinearColor=$NPC.FortniteNPC.EventAudio_Spectrum, Channel:ENiagaraLinearColor_Channels=::NewEnumerator1)
         3:NiagaraFloat=&MakeFloatFromLinearColor(LinearColor=$NPC.FortniteNPC.EventAudio_Spectrum, Channel:ENiagaraLinearColor_Channels=::NewEnumerator2)
         4:NiagaraFloat=&MakeFloatFromLinearColor(LinearColor=$NPC.FortniteNPC.EventAudio_Spectrum, Channel:ENiagaraLinearColor_Channels=::NewEnumerator3)
         5:NiagaraFloat=?
         Alpha:NiagaraFloat=&Sine(
          Normalized Angle=&Add_Float(A=$Emitter.Age, B=&ReturnNormalizedExecIndex())
          Period=1
          Scale=1
          Bias=0
         )
        )
        Input Min=&RandomRangeFloat(Minimum=0.4, Maximum=0.4)
        Input Max=0.85
        Output Min=0.5
        Output Max=1
       )
       B=&Lerp_Float(
        A=0
        B=&MaskFloatBySpawnGroup(Masked Float=1.1, Passthrough Float=1.5)
        Alpha=&RemapRange(
         Input Value To Remap=$NPC.FortniteNPC.EventAudio_Amplitude
         Input Min=0.7
         Input Max=0.85
         Output Min=0.5
         Output Max=&RandomRangeFloat(Minimum=0.9, Maximum=2)
        )
       )
      )
      Alpha=&FloatFromCurve(FloatCurve=DataInterface() {propertyValues {Curve {keys [<at 0.63 0>,<at 1 1>]}}}, CurveIndex=$Emitter.Age)
     )
     Y=1
     Z=1
    )) {module=ScaleMeshSize001}
   &DynamicMaterialParameters(
    Index 0 Param 1=&Multiply_Float(
     A=&RemapRange(Input Value To Remap=$NPC.FortniteNPC.EventAudio_Amplitude, Input Min=0.7, Input Max=0.85, Output Min=0, Output Max=0.1)
     B=&Lerp_Float(
      A=0
      B=&MaskFloatBySpawnGroup(Masked Float=1.1, Passthrough Float=1.5)
      Alpha=&RemapRange(Input Value To Remap=$NPC.FortniteNPC.EventAudio_Amplitude, Input Min=0.65, Input Max=0.88, Output Min=0, Output Max=1)
     )
    )
    Index 0 Param 2=&RandomRangeFloat(Minimum=0.1, Maximum=0.4)
    Index 0 Param 3=&RandomRangeFloat(Minimum=-0.1, Maximum=-0.3)
    Index 0 Param 4=0.2
    Index 1 Param 1=1
    Index 1 Param 2=2
    Index 1 Param 3=0.5
    Index 1 Param 4=1
    Index 2 Param 1=1
    Index 2 Param 4=&MakeCustomFloatFromBool(
     Bool=&SetBoolByIntComparison(A=&ReturnExecIndex(), B=2, Comparison Type=::NewEnumerator0)
     True Float=1
     False Float=0
    )
    Index 3 Param 1=&FloatFromCurve(FloatCurve=DataInterface() {propertyValues {Curve {keys [<at 0 1>,<at 0.6 0.6>]}}}, CurveIndex=$Emitter.Age)
    Module.Index0Param0Write=true
    Module.Index0Param1Write=true
    Module.Index0Param2Write=true
    Module.Index0Param3Write=true
    Module.Index1Param0Write=true
    Module.Index1Param1Write=true
    Module.Index1Param2Write=true
    Module.Index1Param3Write=true
    Module.Index2Param0Write=true
    Module.Index2Param1Write=true
    Module.Index2Param2Write=true
    Module.Index2Param3Write=true
    Module.Index3Param0Write=false
    Module.Index3Param1Write=true
    Module.Index3Param2Write=true
    Module.Index3Param3Write=true
    Write Parameter Index 1=true
    Write Parameter Index 2=true
    Write Parameter Index 3=true
   ) {module=DynamicMaterialParameters001}
   &ScaleColor(
    Scale Alpha=&FloatFromCurve(
     FloatCurve=DataInterface() {propertyValues {Curve {keys [
         {interpMode=RCIM_Cubic, tangentMode=RCTM_User, time=0.3, value=0}
         {arriveTangent=0.5408, interpMode=RCIM_Cubic, leaveTangent=0.5408, tangentMode=RCTM_User, time=1, value=0.73}
        ]}}}
     CurveIndex=$Emitter.Age
    )
    ScaleA=true
    ScaleRGB=true
   )
   &DynamicMaterialParameters(
    Index 3 Param 1=&FloatFromCurve(FloatCurve=DataInterface() {propertyValues {Curve {keys [<at 0 1>,<at 1 0.6>]}}}, CurveIndex=$Emitter.Age)
    Module.Index3Param0Write=true
    Module.Index3Param1Write=false
    Module.Index3Param2Write=false
    Module.Index3Param3Write=false
    Write Parameter Index 0=false
    Write Parameter Index 3=true
   ) {module=DynamicMaterialParameters002}
  ]
  rendererClasses [&NiagaraMeshRendererProperties]
  renderers [{rendererClass=&NiagaraMeshRendererProperties, rendererIndex=0}]
 }
 Pulsing_Squares {
  emitterUpdateScript [
   &EmitterState(Life Cycle Mode=::NewEnumerator1, Loop Duration=1.875, UseLoopDelay=false)
   /SetVariables(Emitter.IsLooping:NiagaraBool=&SetBoolByFloatComparison(A=$Emitter.Age, B=1.5)) {isSetParameters=true}
   &SpawnBurst_Instantaneous(
    Spawn Count=28
    Spawn Probability=&MakeCustomFloatFromBool(Bool=$Emitter.IsLooping, True Float=1, False Float=0)
    Spawn Time=0.9375
    Use Spawn Probability=true
    UseLoopCountLimit=false
   )
   &SpawnBurst_Instantaneous(
    Spawn Count=28
    Spawn Probability=&MakeCustomFloatFromBool(Bool=$Emitter.IsLooping, True Float=0, False Float=1)
    Spawn Time=0.73
    Use Spawn Probability=true
    UseLoopCountLimit=false
   ) {module=SpawnBurst_Instantaneous001}
  ]
  particleSpawnScript [
   &InitializeParticle(
    Color=&SelectLinearColorFromArray(
     Array Sampling Mode=::NewEnumerator1
     Direct Array Index=&RandomRangeInt(Minimum=0, Maximum=3)
     Color Selection Array=DataInterface() {propertyValues {ColorData [<rgba 1 0 0.8627 1>, <rgba 0 1 0.0767 1>, <rgba 0 0.3804 1 1>, <rgba 1 0.9256 0 1>]}}
     Random Seed=?
    )
    Lifetime=0.9375
    Sprite Rotation Angle=90
    Sprite Rotation Mode=::NewEnumerator0
    Sprite Size Mode=::NewEnumerator3
    Uniform Sprite Size=15
    UsePositionOffset=false
    Write Material Random=true
   )
   &MeshRotationForce(Coordinate Space:ENiagaraMeshTransforms=::NewEnumerator2, Rotation=<0 0.8 0>) {enabled=false}
   &ApplyInitialForces()
   &AddVelocity(Module.Use Falloff Distance=false, Override Seed=false, Velocity Mode=::NewEnumerator1, Velocity Speed=0.1)
   &SpriteFacingAndAlignment(Sprite Alignment=$Particles.ShapeLocation.ShapeBitangent, Sprite Facing=<0.01 1 0>, WriteAlignment=true, WriteFacing=true) {enabled=false}
  ]
  particleUpdateScript [
   &ParticleState()
   &ShapeLocation(
    MaskSpawnGroup=false
    Offset=<1 -22 101>
    Offset Mode=::NewEnumerator0
    Override Seed=false
    Ring / Disc Distribution Mode:ENiagaraSphereDistributionMode=::NewEnumerator1
    Ring Radius=68
    Rotation Angle:NiagaraFloat=&Multiply_Float(A=0.016, B=$Engine.System.Age)
    Rotation Axis=&MakeVector(X=0, Y=1, Z=0)
    Rotation Mode=::NewEnumerator1
    Shape Primitive=::NewEnumerator4
    U Position=&ReturnNormalizedExecIndex()
    UseOwnerScale=true
   )
   &ScaleSpriteSize(Uniform Curve Sprite Scale=DataInterface() {propertyValues {Curve {keys [
        {arriveTangent=1.6719, interpMode=RCIM_Cubic, leaveTangent=1.6719, tangentMode=RCTM_User, time=0, value=0.4}
        {arriveTangent=0.0224, interpMode=RCIM_Cubic, leaveTangent=0.0224, tangentMode=RCTM_User, time=1, value=1}
       ]}}})
   &SolveForcesAndVelocity(Clamp Velocity=false, Limit Acceleration=false)
   &AlignSpriteToMeshOrientation(
    Mesh Orientation Relative Sprite Alignment Vector=$Particles.ShapeLocation.ShapeBitangent
    Mesh Orientation Relative Sprite Facing Vector:Vector3f=<0.01 1 0>
    SetAlignment=true
    SetFacing=true
   )
   &DynamicMaterialParameters(
    Index 0 Param 1=&Multiply_Float(
     A=&RemapRange(Input Value To Remap=$NPC.FortniteNPC.EventAudio_Amplitude, Input Min=0.65, Input Max=0.88, Output Min=0.02, Output Max=0.05)
     B=&Lerp_Float(
      A=0
      B=&MaskFloatBySpawnGroup(Masked Float=1.1, Passthrough Float=1.5)
      Alpha=&RemapRange(Input Value To Remap=$NPC.FortniteNPC.EventAudio_Amplitude, Input Min=0.65, Input Max=0.88, Output Min=1, Output Max=1)
     )
    )
    Index 0 Param 2=&Lerp_Float(A=0, B=1, Alpha=&MakeFloatFromInt(Int=$Particles.UpdateGroup))
    Module.Index0Param0Write=true
    Module.Index0Param1Write=true
    Module.Index0Param2Write=true
    Module.Index0Param3Write=true
   ) {module=DynamicMaterialParameters004}
   &ScaleColor(
    Scale Alpha=&FloatFromCurve(FloatCurve=DataInterface() {propertyValues {Curve {keys [
         {arriveTangent=-2.1531, interpMode=RCIM_Cubic, leaveTangent=-2.1531, tangentMode=RCTM_User, time=0, value=1}
         {arriveTangent=-0.0437, interpMode=RCIM_Cubic, leaveTangent=-0.0437, tangentMode=RCTM_User, time=1, value=0}
        ]}}})
    ScaleA=true
    ScaleRGB=true
   )
  ]
  rendererClasses [&NiagaraSpriteRendererProperties]
  renderers [{rendererClass=&NiagaraSpriteRendererProperties}]
 }
}
summary {
 systemName=NS_JadeTowelRing
 emitters [
  {emitterName=Falling_Tris, rendererClasses [&NiagaraMeshRendererProperties]}
  {emitterName=Glow_Body, rendererClasses [&NiagaraMeshRendererProperties]}
  {emitterName=Inner_Ring, rendererClasses [&NiagaraMeshRendererProperties]}
  {emitterName=Pulsing_Squares, rendererClasses [&NiagaraSpriteRendererProperties]}
  {emitterName=Outer_Ring, rendererClasses [&NiagaraMeshRendererProperties]}
  {emitterName=Burst_Ray, rendererClasses [&NiagaraMeshRendererProperties]}
 ]
 userVariables [
  User.bIsFrontEnd {type=NiagaraBool, defaultValue=true}
  User.bIsFrontEndPreview {type=NiagaraBool}
  User.Local Alpha Preview {type=NiagaraFloat, defaultValue=0.4, description=""}
  User.Other Alpha Preview {type=NiagaraFloat, defaultValue=0.4, description=""}
  User.Skeletal Mesh Root {
   type=NiagaraDataInterfaceSkeletalMesh
   defaultValue {
    struct=NiagaraExt_VariableValue_DataInterface
    value {dataInterface=&self:NiagaraDataInterfaceSkeletalMesh_0}
   }
  }
  User.SquareRingColor {type=LinearColor, description="", defaultValue {struct=LinearColor, value=<rgba 0 0.3611 0.7083 1>}}
 ]
}
